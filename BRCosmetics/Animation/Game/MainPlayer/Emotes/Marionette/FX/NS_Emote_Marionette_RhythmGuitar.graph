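# Niagara stage-lighting and spark effect for the rhythm guitar variant of the Marionette emote.
# GroundLightMesh, dome_lights and Light001 form color-cycling ground lights, overhead meshes and a dynamic light, with alpha lerped between two user values by EventAudio_Amplitude.
# SparkleParent001 launches periodic GPU sparks upward, which TrailEmbers001 reads to spawn ember trails.
# Front-end modifiers adjust brightness; the spotlight and two sparkle emitters are disabled.
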NiagaraSystem /BRCosmetics/Animation/Game/MainPlayer/Emotes/Marionette/FX/NS_Emote_Marionette_RhythmGuitar
dependencies {
 usedDataInterfaces [
  &NiagaraDataInterfaceArrayFloat3
  &NiagaraDataInterfaceColorCurve
  &NiagaraDataInterfaceCurve
  &NiagaraDataInterfaceArrayFloat
  /Script/Niagara.NiagaraDataInterfaceVectorField
  &NiagaraDataInterfaceParticleRead
  &NiagaraDataInterfaceVectorCurve
 ]
 usedDynamicInputs []
 usedModules [&SystemState]
}
emitters {
 GroundLightMesh {
  emitterUpdateScript [
   &EmitterState(
    Inactive Response=::NewEnumerator1
    Life Cycle Mode=::NewEnumerator1
    Loop Behavior=::NewEnumerator1
    Loop Duration Mode=::NewEnumerator1
    UseLoopDelay=false
   )
   &SpawnRate(SpawnRate=6, Use Spawn Probability=false)
   &SpawnBurst_Instantaneous(Spawn Count=6, Use Spawn Probability=false, UseLoopCountLimit=false) {enabled=false}
  ]
  particleSpawnScript [
   &FN_SetSpawnGroup(Number of Spawn Groups=6)
   &InitializeParticle(
    Lifetime Max=0.6
    Lifetime Min=0.8
    Lifetime Mode=::NewEnumerator1
    Mesh Scale=&Multiply_Vector(
     A=&SelectVectorFromArray(
      Direct Array Index=$Particles.SpawnGroup
      Vector Selection Array=DataInterface() {propertyValues {FloatData [<0.3 0.3 1.5>, <0.3 0.3 1.8>, <0.25 0.25 1.6>, <0.3 0.3 1.5>, <0.3 0.3 1.7>, <0.2 0.2 1.6>]}}
     )
     B=<1 1 0.8>
    )
    Mesh Scale Mode=::NewEnumerator4
    Position Offset=&Multiply_Vector(
     A=&SelectVectorFromArray(
      Direct Array Index=$Particles.SpawnGroup
      Vector Selection Array=DataInterface() {propertyValues {FloatData [<30 20 2>, <30 -20 2>, <0 -35 2>, <-40 -20 2>, <-40 20 2>, <0 35 2>]}}
     )
     B=<1 1 1>
    )
    Sprite Size Mode=::NewEnumerator3
    Uniform Sprite Size=50
    UsePositionOffset=true
    Write Material Random=true
   )
   &InitialMeshOrientation(
    Mesh Orientation Mode=::NewEnumerator7
    Rotation=&SelectVectorFromArray(
     Direct Array Index=$Particles.SpawnGroup
     Vector Selection Array=DataInterface() {propertyValues {FloatData [<0.2 0.5 -0.05>, <0.2 0.5 0.05>, <0 0.5 0.1>, <0 0.45 0.05>, <0 0.45 -0.05>, <0 0.5 -0.1>]}}
    )
    Use Rotation Vector=true
   )
  ]
  particleUpdateScript [
   &ParticleState()
   &UpdateMeshOrientation(
    Rotation Coordinate Space=::NewEnumerator0
    Rotation Rate=&SelectFloatFromArray(Direct Array Index=$Particles.SpawnGroup, Float Selection Array=DataInterface() {propertyValues {FloatData [0.3,-0.4,0.35,-0.38,0.4,-0.35]}})
    Rotation Vector=<0 0 0.4>
   )
   &Color(Color=&ColorFromCurve(
     ColorCurve=DataInterface() {propertyValues {
       AlphaCurve {keys [<at 0 0.75 RCIM_Cubic>]}
       BlueCurve {keys [<at 0.1 0.015 RCIM_Cubic>,<at 0.15 2>,<at 0.45 2>,<at 0.5 2>,<at 0.65 2>,<at 0.7 0.015>]}
       GreenCurve {keys [<at 0.1 0.1375 RCIM_Cubic>,<at 0.15 0.2583>,<at 0.45 0.2583>,<at 0.5 0.1>,<at 0.65 0.1>,<at 0.7 0.1375>]}
       RedCurve {keys [<at 0.1 0.75 RCIM_Cubic>,<at 0.15 0.1>,<at 0.45 0.1>,<at 0.5 0.2583>,<at 0.65 0.2583>,<at 0.7 0.75>]}
      }}
     CurveIndex=$System.NormalizedLoopAge
    )) {module=Color001}
   &ScaleColor(
    Color Value To Scale=$Particles.Color
    Scale Alpha=&Multiply_Float(
     A=&FloatFromCurve(
      FloatCurve=DataInterface() {propertyValues {Curve {keys [
          {interpMode=RCIM_Cubic, tangentMode=RCTM_User, time=0, value=0}
          {interpMode=RCIM_Cubic, tangentMode=RCTM_User, time=0.3, value=1}
          <at 0.5 0>
          <at 0.8 1>
          {interpMode=RCIM_Cubic, tangentMode=RCTM_User, time=1, value=0}
         ]}}}
      CurveIndex=&Frac_Float(Float=&Add_Float(
        A=$Emitter.Age
        B=&SelectFloatFromArray(Direct Array Index=$Particles.SpawnGroup, Float Selection Array=DataInterface() {propertyValues {FloatData [0.2,0.3,0.5,0.4,0.4,0.18]}})
       ))
      Scale Curve=1.5
     )
     B=&Multiply_Float(
      A=&Lerp_Float(A=$User.Other Alpha Preview, B=$User.Local Alpha Preview, Alpha=$NPC.FortniteNPC.EventAudio_Amplitude)
      B=&FN_FrontEndFloatModifier(Gameplay=1, Front End=2)
     )
    )
    ScaleA=true
    ScaleRGB=true
   )
   &ScaleMeshSize(Scale Factor=&VectorFromCurve(VectorCurve=DataInterface() {propertyValues {XCurve {keys [<at 0 0>,<at 0.2 0.8>]}, YCurve {keys [<at 0 0>,<at 0.2 0.8>]}, ZCurve {keys [<at 0 1>,<at 1 1>]}}}))
   &ScaleSpriteSize(Uniform Curve Sprite Scale=DataInterface() {propertyValues {Curve {keys [{tangentMode=RCTM_User, time=0, value=0}, {tangentMode=RCTM_User, time=0.2, value=1}]}}})
  ]
  rendererClasses [&NiagaraMeshRendererProperties, &NiagaraSpriteRendererProperties]
  renderers [
   {rendererClass=&NiagaraMeshRendererProperties, rendererIndex=0}
   {rendererClass=&NiagaraSpriteRendererProperties, rendererIndex=1}
  ]
 }
 Light001 {
  emitterUpdateScript [
   &EmitterState(
    Inactive Response=::NewEnumerator1
    Life Cycle Mode=::NewEnumerator1
    Loop Behavior=::NewEnumerator1
    Loop Duration Mode=::NewEnumerator1
    UseLoopDelay=false
   )
   &SpawnBurst_Instantaneous(Use Spawn Probability=false, UseLoopCountLimit=false)
  ]
  particleSpawnScript [
   &FN_SetSpawnGroup()
   &InitializeParticle(Mesh Scale Mode=::NewEnumerator3, Position Offset=<0 0 200>, UsePositionOffset=true, Write Material Random=true)
   &CameraOffset()
   /SetVariables(Particles.LightExponent=2, Particles.LightRadius=230) {isSetParameters=true}
   /SetVariables(Particles.AttenuationScale=275) {isSetParameters=true}
  ]
  particleUpdateScript [
   &ParticleState(
    Kill Particles When Lifetime Has Elapsed=false
    Let Infinitely Lived Particles Die When Emitter Deactivates=true
    Loop Particles Lifetime=true
   )
   &Color(Color=&ColorFromCurve(ColorCurve=DataInterface() {propertyValues {
       AlphaCurve {keys [<at 0 0.75 RCIM_Cubic>]}
       BlueCurve {keys [<at 0.1 0.6 RCIM_Cubic>,<at 0.15 4>,<at 0.45 4>,<at 0.5 4>,<at 0.65 4>,<at 0.7 0.6>]}
       GreenCurve {keys [<at 0.1 2.5833 RCIM_Cubic>,<at 0.15 1.8667>,<at 0.45 1.7333>,<at 0.5 0.6>,<at 0.65 0.6>,<at 0.7 2.5833>]}
       RedCurve {keys [<at 0.1 4 RCIM_Cubic>,<at 0.15 0.8>,<at 0.45 0.6>,<at 0.5 1.7333>,<at 0.65 1.7333>,<at 0.7 4>]}
      }})) {module=Color001}
   &ScaleColor(
    Color Value To Scale=$Particles.Color
    Scale Alpha=&Multiply_Float(
     A=&FloatFromCurve(FloatCurve=DataInterface() {propertyValues {Curve {keys [<at 0.1 1>,<at 0.2 0.6>,<at 0.3 1>,<at 0.4 0.4>,<at 0.5 1>,<at 0.7 1>,<at 0.8 0.4>,<at 1 1>]}}})
     B=&Lerp_Float(A=$User.Other Alpha Preview, B=$User.Local Alpha Preview, Alpha=$NPC.FortniteNPC.EventAudio_Amplitude)
    )
    Scale RGB=&VectorFromFloat(Value=30)
    ScaleA=true
    ScaleRGB=true
   )
   /SetVariables(Particles.LightComponentColor:LinearColor=&ColorFromCurve(ColorCurve=DataInterface() {propertyValues {
       AlphaCurve {keys [<at 0 0.75 RCIM_Cubic>]}
       BlueCurve {keys [<at 0.1 0.15 RCIM_Cubic>,<at 0.15 1>,<at 0.45 1>,<at 0.5 1>,<at 0.65 1>,<at 0.7 0.15>]}
       GreenCurve {keys [<at 0.1 0.6458 RCIM_Cubic>,<at 0.15 0.4667>,<at 0.45 0.4333>,<at 0.5 0.15>,<at 0.65 0.15>,<at 0.7 0.6458>]}
       RedCurve {keys [<at 0.1 1 RCIM_Cubic>,<at 0.15 0.2>,<at 0.45 0.15>,<at 0.5 0.4333>,<at 0.65 0.4333>,<at 0.7 1>]}
      }})) {isSetParameters=true}
   /SetVariables(Particles.LightComponentIntensity:NiagaraFloat=&Multiply_Float(
     A=&FloatFromCurve(
      FloatCurve=DataInterface() {propertyValues {Curve {keys [<at 0.1 1>,<at 0.2 0.6>,<at 0.3 1>,<at 0.4 0.4>,<at 0.5 1>,<at 0.7 1>,<at 0.8 0.4>,<at 1 1>]}}}
      Scale Curve=50000
     )
     B=&Lerp_Float(A=$User.Other Alpha Preview, B=$User.Local Alpha Preview, Alpha=$NPC.FortniteNPC.EventAudio_Amplitude)
    )) {isSetParameters=true}
  ]
  rendererClasses [&NiagaraLightRendererProperties, &NiagaraComponentRendererProperties]
  renderers [
   {rendererClass=&NiagaraLightRendererProperties, rendererIndex=0}
   {rendererClass=&NiagaraComponentRendererProperties, rendererIndex=1}
  ]
 }
 SparkleParent001 {
  simTarget=GPUComputeSim
  emitterUpdateScript [
   &EmitterState(Delay First Loop Only=true, Life Cycle Mode=::NewEnumerator1, Loop Delay=0.2, Loop Duration=3.5, UseLoopDelay=true)
   &SpawnRate(
    SpawnRate=&FloatFromCurve(
     FloatCurve=DataInterface() {propertyValues {Curve {keys [{interpMode=RCIM_Cubic, tangentMode=RCTM_User, time=0, value=2}, {interpMode=RCIM_Cubic, tangentMode=RCTM_User, time=0.4, value=0}]}}}
     CurveIndex=$Emitter.NormalizedLoopAge
     Scale Curve=10
    )
    Use Spawn Probability=false
   )
  ]
  particleSpawnScript [
   &InitializeParticle(Lifetime=1.2, Sprite Size Mode=::NewEnumerator3, Uniform Sprite Size=50, UsePositionOffset=false, Write Material Random=true)
   &AddVelocity(
    Cone Angle=1
    Cone Axis=&MakeVector(X=0, Y=&Sine(Normalized Angle=$Emitter.Age, Period=1, Scale=0.4, Bias=0), Z=1)
    Module.UseConeFalloff=false
    Module.UseVelDistribution=false
    Override Seed=false
    Velocity Mode=::NewEnumerator2
    Velocity Speed=270
   )
   &SystemLocation(MaskSpawnGroup=false, Offset=<-50 0 -20>)
  ]
  particleUpdateScript [
   &GravityForce(Gravity=<0 0 -30>)
   &Drag(Use Linear Drag=true, Use Rotational Drag=false)
   &CurlNoiseForce(Noise Frequency=0.25, Noise Strength=15, Pan Noise=false)
   &GravityForce(Gravity=<0 0 -32>) {module=GravityForce001}
   &SolveForcesAndVelocity(Clamp Velocity=false, Limit Acceleration=false)
   &ParticleState()
   &DynamicMaterialParameters(
    Index 0 Param 1=0.8
    Index 0 Param 2=1
    Index 0 Param 3=1
    Index 1 Param 1=150
    Index 1 Param 4=1
    Module.Param0WriteEnabled=true
    Module.Param0WriteEnabled1=true
    Module.Param1WriteEnabled=true
    Module.Param1WriteEnabled1=true
    Module.Param2WriteEnabled=true
    Module.Param2WriteEnabled1=true
    Module.Param3WriteEnabled=false
    Module.Param3WriteEnabled1=true
    Write Parameter Index 1=true
   )
   &Color(Color=&ColorFromCurve(ColorCurve=DataInterface() {propertyValues {
       AlphaCurve {keys [<at 0 0.5 RCIM_Cubic>]}
       BlueCurve {keys [<at 0.1629 0.05 RCIM_Cubic>,<at 1 0 RCIM_Cubic>]}
       GreenCurve {keys [<at 0.1629 0.525 RCIM_Cubic>,<at 1 0 RCIM_Cubic>]}
       RedCurve {keys [<at 0.1629 1 RCIM_Cubic>,<at 1 1 RCIM_Cubic>]}
      }}))
   &ScaleColor(
    Color Value To Scale=$Particles.Color
    Scale Alpha=&FloatFromCurve(
     FloatCurve=DataInterface() {propertyValues {Curve {keys [
         {arriveTangent=9.0725, interpMode=RCIM_Cubic, leaveTangent=9.0725, tangentMode=RCTM_User, time=0, value=0}
         {arriveTangent=-7.2695, interpMode=RCIM_Cubic, leaveTangent=-7.2695, tangentMode=RCTM_User, time=0.2, value=1.75}
         {arriveTangent=-0.0405, interpMode=RCIM_Cubic, leaveTangent=-0.0405, tangentMode=RCTM_User, time=1, value=0}
        ]}}}
     Scale Curve=&Multiply_Float(
      A=0.1
      B=&FloatFromCurve(
       FloatCurve=DataInterface() {propertyValues {Curve {keys [{interpMode=RCIM_Cubic, tangentMode=RCTM_User, time=0, value=1}, {interpMode=RCIM_Cubic, tangentMode=RCTM_User, time=1, value=0}]}}}
       Scale Curve=&Waveform(
        Global Amplitude Scale=5
        "Waveform [1]"=::NewEnumerator0
        "[1] Phase Scale"=&RandomRangeFloat(Minimum=3, Maximum=5)
        "[1] Frequency"=1
        "[1] Amplitude Min/Max"=<0.5 1>
       )
      )
     )
    )
    Scale RGB=&VectorFromFloat(Value=&FN_FrontEndFloatModifier(Gameplay=25, Front End=5))
    ScaleA=true
    ScaleRGB=true
   ) {module=ScaleColor001}
   &ScaleSpriteSize(
    Uniform Curve Scale=0.25
    Uniform Curve Sprite Scale=DataInterface() {propertyValues {Curve {keys [
        {arriveTangent=2.9868, interpMode=RCIM_Cubic, leaveTangent=2.9868, tangentMode=RCTM_User, time=0, value=1.125}
        <at 0.1 2.2>
        <at 0.8 2.3>
        {arriveTangent=-0.0405, interpMode=RCIM_Cubic, leaveTangent=-0.0405, tangentMode=RCTM_User, time=1, value=1.4}
       ]}}}
   )
   /Niagara/VectorFields/SampleVectorField(
    FieldRotate:Quat4f=<0 0 1 1>
    Sampled Vector Scale=0.05
    VectorField=DataInterface() {propertyValues {Field=/Game/Effects/VectorFields/Noisy_Turbulence_103, bTileX=true, bTileY=true, bTileZ=true}}
   )
   /Niagara/VectorFields/ApplyVectorField(Vector Field Intensity as Force=0, Vector Field Intensity as Velocity=50, Vector Field Velocity Tightness=0.001)
  ]
  rendererClasses [&NiagaraSpriteRendererProperties]
  renderers [{rendererClass=&NiagaraSpriteRendererProperties}]
 }
 Sparklys {
  bEnabled=false
  emitterUpdateScript [
   &EmitterState(
    Inactive Response=::NewEnumerator1
    Life Cycle Mode=::NewEnumerator1
    Loop Duration=&RandomRangeFloat(Minimum=3, Maximum=6)
    UseLoopDelay=false
   )
   &SpawnRate(SpawnRate=20, Use Spawn Probability=false)
   &SpawnBurst_Instantaneous(
    Spawn Count=&RandomRangeInt(Minimum=1, Maximum=2)
    Spawn Time=&RandomRangeFloat(Minimum=0, Maximum=1.5)
    Use Spawn Probability=false
    UseLoopCountLimit=false
   )
  ]
  particleSpawnScript [
   &InitializeParticle(
    Color=&MakeCustomLinearColorFromBool(
     Bool=&RandomBool()
     True Color=&MakeCustomLinearColorFromBool(Bool=&RandomBool(), True Color=<rgba 0.2561 0.6588 1 1>, False Color=<rgba 1 0.3829 0.2596 1>)
     False Color=&MakeCustomLinearColorFromBool(Bool=&RandomBool(), True Color=<rgba 0.9047 0.3204 1 1>, False Color=<rgba 1 0.4905 0.0625 1>)
    )
    Lifetime Max=2.4
    Lifetime Min=1.2
    Lifetime Mode=::NewEnumerator1
    Sprite Rotation Mode=::NewEnumerator2
    Sprite Size Mode=::NewEnumerator1
    Uniform Sprite Size Max=4
    Uniform Sprite Size Min=1
    UsePositionOffset=false
    Write Material Random=true
   )
   &ShapeLocation(
    MaskSpawnGroup=false
    Offset=<0 0 50>
    Offset Mode=::NewEnumerator0
    Override Seed=false
    Shape Origin=<0 0 0>
    Sphere Radius=50
    UseOwnerScale=true
   )
   &AddVelocity(Velocity=&RandomRangeVector(Minimum=<30 30 40>, Maximum=<-30 -30 70>))
   &ApplyInitialForces(Apply Force to Position=true)
   /SetVariables(Particles.Curve:NiagaraFloat=&FloatFromCurve(FloatCurve=DataInterface() {propertyValues {Curve {keys [
         {interpMode=RCIM_Cubic, leaveTangent=7.0743, tangentMode=RCTM_Break, time=0, value=0.35}
         {interpMode=RCIM_Cubic, tangentMode=RCTM_Break, time=0.2, value=0.95}
         {interpMode=RCIM_Cubic, leaveTangent=-0.1241, tangentMode=RCTM_Break, time=0.4, value=0.55}
         {interpMode=RCIM_Cubic, tangentMode=RCTM_Break, time=0.6, value=0.8}
         {interpMode=RCIM_Cubic, tangentMode=RCTM_Break, time=0.8, value=0.6}
         {arriveTangent=2.6513, interpMode=RCIM_Cubic, tangentMode=RCTM_Break, time=1, value=1}
        ]}}})) {isSetParameters=true}
  ]
  particleUpdateScript [
   &ParticleState()
   &Drag(Drag=2.5, Use Linear Drag=true, Use Rotational Drag=false)
   &CurlNoiseForce(
    Noise Frequency=100
    Noise Strength=&FloatFromCurve(FloatCurve=DataInterface() {propertyValues {Curve {keys [<at 0 0>,<at 1 40>]}}}, Scale Curve=10)
    Pan Noise=false
   )
   &GravityForce(Gravity=&RandomRangeVector(Minimum=<0 0 200>, Maximum=<0 0 20>))
   &SolveForcesAndVelocity(Clamp Velocity=false, Limit Acceleration=false)
   &DynamicMaterialParameters(
    Index 0 Param 1=&Multiply_Float(
     A=&FloatFromCurve(FloatCurve=DataInterface() {propertyValues {Curve {keys [
          {interpMode=RCIM_Cubic, leaveTangent=0.1514, tangentMode=RCTM_Break, time=0, value=1}
          {arriveTangent=-0.3559, interpMode=RCIM_Cubic, tangentMode=RCTM_Break, time=1, value=0}
         ]}}})
     B=&Lerp_Float(A=1, B=0, Alpha=$Particles.Curve)
    )
    Module.Param0WriteEnabled=true
    Module.Param1WriteEnabled=true
    Module.Param2WriteEnabled=true
    Module.Param3WriteEnabled=true
   )
   &ScaleColor(
    Scale Alpha=&FloatFromCurve(
     FloatCurve=DataInterface() {propertyValues {Curve {keys [
         {arriveTangent=9.8157, interpMode=RCIM_Cubic, leaveTangent=9.8157, tangentMode=RCTM_User, time=0, value=0}
         {arriveTangent=-1.4617, interpMode=RCIM_Cubic, leaveTangent=-1.4617, tangentMode=RCTM_User, time=0.2, value=1}
         {interpMode=RCIM_Cubic, tangentMode=RCTM_Break, time=1, value=0}
        ]}}}
     Scale Curve=5
    )
    Scale RGB=&VectorFromFloat(Value=&Lerp_Float(
      A=1
      B=5
      Alpha=&Sine(Normalized Angle=$Emitter.Age, Period=&RandomRangeFloat(Minimum=0.1, Maximum=0.4), Scale=1, Bias=1)
     ))
    ScaleA=true
    ScaleRGB=true
   )
   &ScaleSpriteSize(
    Scale Sprite Size Mode=::NewEnumerator0
    Uniform Scale Factor=&Multiply_Float(
     A=&FloatFromCurve(FloatCurve=DataInterface() {propertyValues {Curve {keys [
          {interpMode=RCIM_Cubic, leaveTangent=2.1941, tangentMode=RCTM_Break, time=0, value=0.35}
          {arriveTangent=0.204, interpMode=RCIM_Cubic, tangentMode=RCTM_Break, time=1, value=1}
         ]}}})
     B=$Particles.Curve
    )
   )
   /SetVariables(
    Particles.New Color:LinearColor=&MultiplyLinearColor(
     Linear Color=$Particles.Color
     Scale Factor=&MakeFloat4FromVectorAndFloat("Vector (RGB)"=&VectorFromFloat(Value=0.5), "Float (Alpha)"=1)
    )
    Particles.New Position:NiagaraPosition=&Add_Vector(
     A=Linked() {linkedVariable {name=Particles.Position, type=NiagaraPosition}}
     B=Linked() {linkedVariable {name=Engine.Owner.Position, type=NiagaraPosition}}
    )
    Particles.New Size:Vector2f=&Multiply_Vector2D(
     A=$Particles.SpriteSize
     B=&Vector2DFromFloat(Value=&RandomRangeFloat(Minimum=4, Maximum=10))
    )
   ) {isSetParameters=true, module=SetVariables_#1}
  ]
  rendererClasses [&NiagaraSpriteRendererProperties]
  renderers [
   {rendererClass=&NiagaraSpriteRendererProperties}
   {rendererClass=&NiagaraSpriteRendererProperties, rendererIndex=1}
  ]
 }
 Sparklys002 {
  bEnabled=false
  emitterUpdateScript [
   &EmitterState(
    Inactive Response=::NewEnumerator1
    Life Cycle Mode=::NewEnumerator1
    Loop Duration=&RandomRangeFloat(Minimum=1, Maximum=0.5)
    UseLoopDelay=false
   )
   &SpawnRate(SpawnRate=3, Use Spawn Probability=false)
   &SpawnBurst_Instantaneous(
    Spawn Count=&RandomRangeInt(Minimum=1, Maximum=3)
    Spawn Time=&RandomRangeFloat(Minimum=0, Maximum=0.5)
    Use Spawn Probability=false
    UseLoopCountLimit=false
   )
  ]
  particleSpawnScript [
   &InitializeParticle(
    Color=&RandomRangeLinearColor(Minimum=<rgba 1 0.3 1 1>, Maximum=<rgba 0.3 0.4167 1 1>)
    Lifetime Max=0.15
    Lifetime Min=0.2
    Lifetime Mode=::NewEnumerator1
    Sprite Size Mode=::NewEnumerator1
    Uniform Sprite Size Max=150
    Uniform Sprite Size Min=70
    UsePositionOffset=false
    Write Material Random=true
   )
   &ShapeLocation(
    MaskSpawnGroup=false
    Non Uniform Scale=<1 1 1.2>
    Offset=<0 0 150>
    Offset Mode=::NewEnumerator0
    Override Seed=false
    Shape Origin=<0 0 0>
    Sphere Radius=50
    UseOwnerScale=true
   )
   &CameraOffset(Camera Offset Amount=-100)
   /SetVariables(Particles.New Visibility:NiagaraInt32=&MakeIntFromFloat(Float=&UniformAOrBFloat(A=0, Distribution Weight=0.3))) {isSetParameters=true}
  ]
  particleUpdateScript [
   &ParticleState()
   &ScaleColor(
    Scale Alpha=&FloatFromCurve(FloatCurve=DataInterface() {propertyValues {Curve {keys [
         {arriveTangent=9.8157, interpMode=RCIM_Cubic, leaveTangent=9.8157, tangentMode=RCTM_User, time=0, value=0}
         {arriveTangent=-1.4617, interpMode=RCIM_Cubic, leaveTangent=-1.4617, tangentMode=RCTM_User, time=0.2, value=1}
         {interpMode=RCIM_Cubic, tangentMode=RCTM_User, time=0.3, value=0}
         {interpMode=RCIM_Cubic, tangentMode=RCTM_User, time=0.5, value=0.75}
         {interpMode=RCIM_Cubic, tangentMode=RCTM_Break, time=1, value=0}
        ]}}})
    ScaleA=true
    ScaleRGB=true
   )
   &ScaleSpriteSize(Uniform Curve Sprite Scale=DataInterface() {propertyValues {Curve {keys [{tangentMode=RCTM_User, time=0, value=1}, {tangentMode=RCTM_User, time=1, value=0}]}}})
   /SetVariables(
    Particles.New Position:NiagaraPosition=&Add_Vector(
     A=Linked() {linkedVariable {name=Particles.Position, type=NiagaraPosition}}
     B=Linked() {linkedVariable {name=Engine.Owner.Position, type=NiagaraPosition}}
    )
    Particles.New Rotation=45
    Particles.New Size:Vector2f=&Multiply_Vector2D(A=$Particles.SpriteSize, B=<0.5 0.5>)
   ) {isSetParameters=true, module=SetVariables_#1}
  ]
  rendererClasses [&NiagaraSpriteRendererProperties]
  renderers [
   {rendererClass=&NiagaraSpriteRendererProperties}
   {rendererClass=&NiagaraSpriteRendererProperties, rendererIndex=1}
  ]
 }
 Sparklys_Falling_LWE {rendererClasses [], renderers []}
 Sparklys_LWE {rendererClasses [], renderers []}
 SpotLight_Main002 {
  bEnabled=false
  emitterUpdateScript [
   &EmitterState(
    Inactive Response=::NewEnumerator1
    Life Cycle Mode=::NewEnumerator1
    Loop Behavior=::NewEnumerator1
    Loop Duration Mode=::NewEnumerator1
    UseLoopDelay=false
   )
   &SpawnBurst_Instantaneous(Use Spawn Probability=false, UseLoopCountLimit=false)
  ]
  particleSpawnScript [
   &FN_SetSpawnGroup()
   &InitializeParticle(Mesh Scale=<0.9 0.9 1.2>, Mesh Scale Mode=::NewEnumerator4, Position Offset=<0 0 280>, UsePositionOffset=true, Write Material Random=true)
   /SetVariables(Particles.Position002:NiagaraPosition=<0 0 3>) {isSetParameters=true}
   /SetVariables(Particles.Position003:NiagaraPosition=<0 0 80>) {isSetParameters=true}
   /SetVariables(Particles.Size002:Vector3f=<0.4 0.4 1.5>) {isSetParameters=true}
   /SetVariables(Particles.Size003:Vector3f=&VectorFromFloat(Value=0.4)) {isSetParameters=true}
   /SetVariables(Particles.Size004:Vector3f=&VectorFromFloat(Value=0.45)) {isSetParameters=true}
   /SetVariables(Particles.Size005:Vector2f=<300 500>) {isSetParameters=true}
   &DynamicMaterialParameters(
    Index 0 Param 1=10
    Index 0 Param 2=1
    Index 0 Param 3=1
    Index 0 Param 4=2
    Index 1 Param 1=15
    Module.Param0WriteEnabled=true
    Module.Param0WriteEnabled1=true
    Module.Param1WriteEnabled=true
    Module.Param1WriteEnabled1=true
    Module.Param2WriteEnabled=true
    Module.Param2WriteEnabled1=true
    Module.Param3WriteEnabled=true
    Module.Param3WriteEnabled1=true
    Write Parameter Index 1=true
   )
  ]
  particleUpdateScript [
   &ParticleState(
    Kill Particles When Lifetime Has Elapsed=false
    Let Infinitely Lived Particles Die When Emitter Deactivates=true
    Loop Particles Lifetime=true
   )
   &Color(Color=&ColorFromCurve(ColorCurve=DataInterface() {propertyValues {
       AlphaCurve {keys [<at 0 0.75 RCIM_Cubic>]}
       BlueCurve {keys [<at 0.1 0.6 RCIM_Cubic>,<at 0.15 4>,<at 0.45 4>,<at 0.5 4>,<at 0.65 4>,<at 0.7 0.6>]}
       GreenCurve {keys [<at 0.1 2.5833 RCIM_Cubic>,<at 0.15 1.8667>,<at 0.45 1.7333>,<at 0.5 0.6>,<at 0.65 0.6>,<at 0.7 2.5833>]}
       RedCurve {keys [<at 0.1 4 RCIM_Cubic>,<at 0.15 0.8>,<at 0.45 0.6>,<at 0.5 1.7333>,<at 0.65 1.7333>,<at 0.7 4>]}
      }})) {module=Color001}
   &ScaleColor(
    Color Value To Scale=$Particles.Color
    Scale Alpha=&Multiply_Float(
     A=&FloatFromCurve(FloatCurve=DataInterface() {propertyValues {Curve {keys [<at 0.1 1>,<at 0.2 0.6>,<at 0.3 1>,<at 0.4 0.4>,<at 0.5 1>,<at 0.7 1>,<at 0.8 0.4>,<at 1 1>]}}})
     B=&Multiply_Float(
      A=&Lerp_Float(A=$User.Other Alpha Preview, B=$User.Local Alpha Preview, Alpha=$NPC.FortniteNPC.EventAudio_Amplitude)
      B=&FN_FrontEndFloatModifier(Gameplay=1, Front End=1.5)
     )
    )
    Scale RGB=&VectorFromFloat(Value=1.5)
    ScaleA=true
    ScaleRGB=true
   )
   /SetVariables(Particles.Color002=&MakeLinearColorFromVectorAndFloat(
     "Vector (RGB)"=&MakeVectorFromLinearColorRGB(Color=&ColorFromCurve(ColorCurve=DataInterface() {propertyValues {
         AlphaCurve {keys [<at 0 0.75 RCIM_Cubic>]}
         BlueCurve {keys [<at 0.1 0.4 RCIM_Cubic>,<at 0.15 4>,<at 0.45 4>,<at 0.5 4>,<at 0.65 4>,<at 0.7 0.4>]}
         GreenCurve {keys [<at 0.1 2.5 RCIM_Cubic>,<at 0.15 1.4667>,<at 0.45 1.4667>,<at 0.5 0.08>,<at 0.65 0.08>,<at 0.7 2.5>]}
         RedCurve {keys [<at 0.1 4 RCIM_Cubic>,<at 0.15 0.2>,<at 0.45 0.2>,<at 0.5 1.3867>,<at 0.65 1.3867>,<at 0.7 4>]}
        }}))
     "Float (Alpha)"=&Multiply_Float(
      A=&FloatFromCurve(
       FloatCurve=DataInterface() {propertyValues {Curve {keys [<at 0.1 1>,<at 0.2 0.2>,<at 0.3 1>,<at 0.4 0>,<at 0.5 1>,<at 0.7 1>,<at 0.8 0>,<at 1 1>]}}}
       Scale Curve=0.75
      )
      B=&Multiply_Float(
       A=&Lerp_Float(A=$User.Other Alpha Preview, B=$User.Local Alpha Preview, Alpha=$NPC.FortniteNPC.EventAudio_Amplitude)
       B=&FN_FrontEndFloatModifier(Gameplay=1, Front End=1.5)
      )
     )
    )) {isSetParameters=true}
   /SetVariables(Particles.Color003:LinearColor=&MakeLinearColorFromVectorAndFloat(
     "Vector (RGB)"=&MakeVectorFromLinearColorRGB(Color=&MultiplyLinearColor(
       Linear Color=<rgba 2 0.8917 0.1 1>
       Scale Factor=&MakeVector4_FromLinearColor(LinearColor=&ColorFromCurve(ColorCurve=DataInterface() {propertyValues {
           AlphaCurve {keys [<at 0 0.75 RCIM_Cubic>]}
           BlueCurve {keys [<at 0.1 0.6 RCIM_Cubic>,<at 0.15 4>,<at 0.45 4>,<at 0.5 4>,<at 0.65 4>,<at 0.7 0.6>]}
           GreenCurve {keys [<at 0.1 2.5833 RCIM_Cubic>,<at 0.15 1.8667>,<at 0.45 1.7333>,<at 0.5 0.6>,<at 0.65 0.6>,<at 0.7 2.5833>]}
           RedCurve {keys [<at 0.1 4 RCIM_Cubic>,<at 0.15 0.8>,<at 0.45 0.6>,<at 0.5 1.7333>,<at 0.65 1.7333>,<at 0.7 4>]}
          }}))
      ))
     "Float (Alpha)"=&Multiply_Float(
      A=&FloatFromCurve(
       FloatCurve=DataInterface() {propertyValues {Curve {keys [<at 0.1 1>,<at 0.2 0.2>,<at 0.3 1>,<at 0.4 0>,<at 0.5 1>,<at 0.7 1>,<at 0.8 0>,<at 1 1>]}}}
       Scale Curve=0.25
      )
      B=&Multiply_Float(
       A=&Lerp_Float(A=$User.Other Alpha Preview, B=$User.Local Alpha Preview, Alpha=$NPC.FortniteNPC.EventAudio_Amplitude)
       B=&FN_FrontEndFloatModifier(Gameplay=1, Front End=5)
      )
     )
    )) {isSetParameters=true}
   /SetVariables(Particles.Color004:LinearColor=&MakeLinearColorFromVectorAndFloat(
     "Vector (RGB)"=&MakeVectorFromLinearColorRGB(Color=&MultiplyLinearColor(
       Linear Color=&ColorFromCurve(ColorCurve=DataInterface() {propertyValues {
          AlphaCurve {keys [<at 0 0.75 RCIM_Cubic>]}
          BlueCurve {keys [<at 0.1 0.6 RCIM_Cubic>,<at 0.15 4>,<at 0.45 4>,<at 0.5 4>,<at 0.65 4>,<at 0.7 0.6>]}
          GreenCurve {keys [<at 0.1 2.5833 RCIM_Cubic>,<at 0.15 1.8667>,<at 0.45 1.7333>,<at 0.5 0.6>,<at 0.65 0.6>,<at 0.7 2.5833>]}
          RedCurve {keys [<at 0.1 4 RCIM_Cubic>,<at 0.15 0.8>,<at 0.45 0.6>,<at 0.5 1.7333>,<at 0.65 1.7333>,<at 0.7 4>]}
         }})
       Scale Factor=&MakeVector4_FromFloat(Float=0.01)
      ))
     "Float (Alpha)"=&Multiply_Float(
      A=&FloatFromCurve(
       FloatCurve=DataInterface() {propertyValues {Curve {keys [<at 0.1 1>,<at 0.2 0.2>,<at 0.3 1>,<at 0.4 0>,<at 0.5 1>,<at 0.7 1>,<at 0.8 0>,<at 1 1>]}}}
       Scale Curve=&MakeCustomFloatFromBool(Bool=$User.bIsFrontEndPreview, True Float=0.3, False Float=0)
      )
      B=&Multiply_Float(A=70, B=&FN_FrontEndFloatModifier(Gameplay=0, Front End=1))
     )
    )) {isSetParameters=true}
   /SetVariables(Particles.Color005:LinearColor=&MakeLinearColorFromVectorAndFloat(
     "Vector (RGB)"=&MakeVectorFromLinearColorRGB(Color=<rgba 0 0 0 1>)
     "Float (Alpha)"=&FN_FrontEndFloatModifier(Gameplay=0, Front End=0.5)
    )) {isSetParameters=true}
  ]
  rendererClasses [&NiagaraMeshRendererProperties, &NiagaraSpriteRendererProperties]
  renderers [
   {rendererClass=&NiagaraMeshRendererProperties, rendererIndex=0}
   {rendererClass=&NiagaraMeshRendererProperties, rendererIndex=1}
   {rendererClass=&NiagaraMeshRendererProperties, rendererIndex=2}
   {rendererClass=&NiagaraMeshRendererProperties, rendererIndex=3}
   {rendererClass=&NiagaraSpriteRendererProperties, rendererIndex=4}
  ]
 }
 TrailEmbers001 {
  simTarget=GPUComputeSim
  emitterUpdateScript [
   &EmitterState()
   &SpawnParticlesFromOtherEmitter(
    Attribute Reader=DataInterface() {propertyValues {EmitterBinding {bindingMode=Other, emitterName=SparkleParent001}}}
    Maximum Age Spawn Threshold=2
    Minimum Age Spawn Threshold=0
    Spawn Rate=100
    Use Maximum Age=false
    Use Minimum Age=false
   )
  ]
  particleSpawnScript [
   &InitializeParticle(
    Lifetime Max=1.4
    Lifetime Min=0.8
    Lifetime Mode=::NewEnumerator1
    Sprite Rotation Mode=::NewEnumerator2
    Sprite Size Mode=::NewEnumerator1
    UsePositionOffset=false
    Write Material Random=true
   )
   &SampleParticlesFromOtherEmitter(
    Applied Velocity Scale:NiagaraFloat=&FloatFromCurve(FloatCurve=DataInterface() {propertyValues {Curve {keys [<at 0 1>,<at 1 0>]}}}, Scale Curve=0.5)
    Maximum Sampled Particle Age=2
    Minimum Sampled Particle Age=0
    Module.MaxSampledAgeCutoff=true
    Module.MinSampledAgeCutoff=true
    Reject Particle By Sampled Age=true
    Velocity Sampling=::NewEnumerator0
   )
   &AddVelocity(
    Cone Angle=1
    Cone Axis=&MakeVector(X=&Sine(Normalized Angle=$Emitter.Age, Period=5, Scale=1, Bias=0), Y=0, Z=1)
    Module.UseConeFalloff=false
    Module.UseVelDistribution=false
    Override Seed=false
    Velocity Mode=::NewEnumerator2
    Velocity Speed=5
   )
   &SphereLocation(
    MaskSpawnGroup=false
    Override Seed=false
    Sphere Radius=2
    Surface Only=false
    UseNonUniformScale=false
    UseOffset=false
    UseSphereOrientation=false
   )
  ]
  particleUpdateScript [
   &ParticleState()
   &Drag(Drag=2, Use Linear Drag=true, Use Rotational Drag=false)
   &GravityForce(Gravity=<0 0 -35>)
   &CurlNoiseForce(Noise Frequency=1, Noise Strength=15, Pan Noise=false)
   &SolveForcesAndVelocity(Clamp Velocity=false, Limit Acceleration=false)
   &Color(Color=&ColorFromCurve(ColorCurve=DataInterface() {propertyValues {
       AlphaCurve {keys [<at 0 1 RCIM_Cubic>]}
       BlueCurve {keys [<at 0.0079 3>,<at 0.1199 0>,<at 0.1889 0>,<at 0.3054 1>,<at 1 1 RCIM_Cubic>]}
       GreenCurve {keys [<at 0.0079 21>,<at 0.1199 0.4167>,<at 0.1889 0>,<at 0.3054 0>,<at 1 0 RCIM_Cubic>]}
       RedCurve {keys [<at 0.0079 30>,<at 0.1199 1>,<at 0.1889 1>,<at 0.3054 0>,<at 1 0.6257 RCIM_Cubic>]}
      }}))
   &ScaleColor(
    Color Value To Scale=$Particles.Color
    Scale Alpha=&FloatFromCurve(
     FloatCurve=DataInterface() {propertyValues {Curve {keys [
         {arriveTangent=9.0725, interpMode=RCIM_Cubic, leaveTangent=9.0725, tangentMode=RCTM_User, time=0, value=0}
         {arriveTangent=-7.2695, interpMode=RCIM_Cubic, leaveTangent=-7.2695, tangentMode=RCTM_User, time=0.2, value=1.75}
         {arriveTangent=-0.0405, interpMode=RCIM_Cubic, leaveTangent=-0.0405, tangentMode=RCTM_User, time=1, value=0}
        ]}}}
     Scale Curve=&Multiply_Float(
      A=0.1
      B=&FloatFromCurve(
       FloatCurve=DataInterface() {propertyValues {Curve {keys [{interpMode=RCIM_Cubic, tangentMode=RCTM_User, time=0, value=1}, {interpMode=RCIM_Cubic, tangentMode=RCTM_User, time=1, value=0}]}}}
       Scale Curve=&Waveform(
        Global Amplitude Scale=5
        "Waveform [1]"=::NewEnumerator0
        "[1] Phase Scale"=&RandomRangeFloat(Minimum=3, Maximum=10)
        "[1] Frequency"=1
        "[1] Amplitude Min/Max"=<0 1>
       )
      )
     )
    )
    Scale RGB=&Multiply_VectorByFloat(Vector=<25 25 25>, Float=&FN_FrontEndFloatModifier(Gameplay=1, Front End=0.15))
    ScaleA=true
    ScaleRGB=true
   ) {module=ScaleColor001}
   &ScaleSpriteSize(
    Uniform Curve Scale=0.075
    Uniform Curve Sprite Scale=DataInterface() {propertyValues {Curve {keys [
        {arriveTangent=2.9868, interpMode=RCIM_Cubic, leaveTangent=2.9868, tangentMode=RCTM_User, time=0, value=1.125}
        <at 0.4 2.3>
        {arriveTangent=-0.0405, interpMode=RCIM_Cubic, leaveTangent=-0.0405, tangentMode=RCTM_User, time=1, value=1.4}
       ]}}}
   )
  ]
  rendererClasses [&NiagaraSpriteRendererProperties]
  renderers [{rendererClass=&NiagaraSpriteRendererProperties}]
 }
 dome_lights {
  emitterUpdateScript [
   &EmitterState(Inactive Response=::NewEnumerator1, Life Cycle Mode=::NewEnumerator1, Loop Duration=0.75, UseLoopDelay=false)
   &SpawnRate(SpawnRate=5, Use Spawn Probability=false)
  ]
  particleSpawnScript [
   &FN_SetSpawnGroup(Number of Spawn Groups=3)
   &InitializeParticle(
    Color=<rgba 1 0.75 0.4 0.1>
    Lifetime Min=2.5
    Lifetime Mode=::NewEnumerator1
    Mesh Scale=&SelectVectorFromArray(
     Direct Array Index=$Particles.SpawnGroup
     Vector Selection Array=DataInterface() {propertyValues {FloatData [<0.6 0.6 1.4>, <0.6 0.6 1.4>, <0.6 0.6 1.4>]}}
    )
    Mesh Scale Mode=::NewEnumerator4
    Position Offset=&SelectVectorFromArray(
     Direct Array Index=$Particles.SpawnGroup
     Vector Selection Array=DataInterface() {propertyValues {FloatData [<-140 40 250>, <-50 120 250>, <-50 -120 250>]}}
    )
    UsePositionOffset=true
    Write Material Random=true
   )
   &InitialMeshOrientation(
    Mesh Orientation Mode=::NewEnumerator7
    Rotation=&SelectVectorFromArray(
     Direct Array Index=$Particles.SpawnGroup
     Vector Selection Array=DataInterface() {propertyValues {FloatData [<-0.02 -0.06 0>, <-0.06 -0.03 0>, <0.06 -0.03 0>]}}
    )
    Use Rotation Vector=true
   )
   &CameraOffset(Camera Offset Amount=-20)
  ]
  particleUpdateScript [
   &ParticleState()
   &Color(Color=&ColorFromCurve(
     ColorCurve=DataInterface() {propertyValues {
       AlphaCurve {keys [<at 0 0.75 RCIM_Cubic>]}
       BlueCurve {keys [<at 0.1 0.6 RCIM_Cubic>,<at 0.15 4>,<at 0.45 4>,<at 0.5 4>,<at 0.65 4>,<at 0.7 0.6>]}
       GreenCurve {keys [<at 0.1 2.5833 RCIM_Cubic>,<at 0.15 1.8667>,<at 0.45 1.7333>,<at 0.5 0.6>,<at 0.65 0.6>,<at 0.7 2.5833>]}
       RedCurve {keys [<at 0.1 4 RCIM_Cubic>,<at 0.15 0.8>,<at 0.45 0.6>,<at 0.5 1.7333>,<at 0.65 1.7333>,<at 0.7 4>]}
      }}
     CurveIndex=$Emitter.NormalizedLoopAge
    )) {module=Color001}
   &ScaleColor(
    Color Value To Scale=$Particles.Color
    Scale Alpha=&Multiply_Float(
     A=&FloatFromCurve(
      FloatCurve=DataInterface() {propertyValues {Curve {keys [
          {interpMode=RCIM_Cubic, tangentMode=RCTM_User, time=0, value=0}
          {interpMode=RCIM_Cubic, tangentMode=RCTM_User, time=0.5, value=1}
          {tangentMode=RCTM_User, time=1, value=0}
         ]}}}
      Scale Curve=0.025
     )
     B=&Multiply_Float(
      A=&Lerp_Float(A=$User.Other Alpha Preview, B=$User.Local Alpha Preview, Alpha=$NPC.FortniteNPC.EventAudio_Amplitude)
      B=&FN_FrontEndFloatModifier(Gameplay=1, Front End=2)
     )
    )
    Scale RGB=&VectorFromFloat(Value=&FloatFromCurve(
      FloatCurve=DataInterface() {propertyValues {Curve {keys [
          {interpMode=RCIM_Cubic, tangentMode=RCTM_Break, time=0.4, value=1}
          {interpMode=RCIM_Cubic, tangentMode=RCTM_Break, time=0.5, value=3}
          {interpMode=RCIM_Cubic, tangentMode=RCTM_Break, time=0.6, value=1}
          <at 0.77 1>
          <at 0.82 1>
         ]}}}
      CurveIndex=&Add_Float(
       A=$Particles.NormalizedAge
       B=&SelectFloatFromArray(
        Direct Array Index=&SpawnGroup()
        Float Selection Array=DataInterface() {propertyValues {FloatData [0.03,0.04,0.025,0.02,0.035,0.027]}}
       )
      )
     ))
    ScaleA=true
    ScaleRGB=true
   )
  ]
  rendererClasses [&NiagaraMeshRendererProperties]
  renderers [
   {rendererClass=&NiagaraMeshRendererProperties, rendererIndex=0}
   {rendererClass=&NiagaraMeshRendererProperties, rendererIndex=1}
  ]
 }
}
summary {
 systemName=NS_Emote_Marionette_RhythmGuitar
 emitters [
  {emitterName=dome_lights, rendererClasses [&NiagaraMeshRendererProperties]}
  {emitterName=Light001, rendererClasses [&NiagaraLightRendererProperties, &NiagaraComponentRendererProperties]}
  {emitterName=SparkleParent001, simTarget=GPUComputeSim, rendererClasses [&NiagaraSpriteRendererProperties]}
  {emitterName=TrailEmbers001, simTarget=GPUComputeSim, rendererClasses [&NiagaraSpriteRendererProperties]}
  {
   emitterName=GroundLightMesh
   rendererClasses [&NiagaraMeshRendererProperties, &NiagaraSpriteRendererProperties]
  }
  {bEnabled=false, emitterName=Sparklys, rendererClasses [&NiagaraSpriteRendererProperties]}
  {bEnabled=false, emitterName=Sparklys002, rendererClasses [&NiagaraSpriteRendererProperties]}
  {
   bEnabled=false
   emitterName=SpotLight_Main002
   rendererClasses [&NiagaraMeshRendererProperties, &NiagaraSpriteRendererProperties]
  }
  {emitterName=Sparklys_Falling_LWE, rendererClasses []}
  {emitterName=Sparklys_LWE, rendererClasses []}
 ]
 userVariables [
  User.bIsFrontEnd {type=NiagaraBool}
  User.bIsFrontEndPreview {type=NiagaraBool}
  User.Local Alpha Preview {type=NiagaraFloat, defaultValue=0.75, description=""}
  User.Other Alpha Preview {type=NiagaraFloat, defaultValue=0.25, description=""}
 ]
}
